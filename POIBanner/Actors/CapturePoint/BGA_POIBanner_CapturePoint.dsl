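; POI banner capture point: a team alone in the capture zone advances progress toward raising the linked banner.
; The server tracks overlapping pawns by team to switch between empty, uncontested, contested and captured states, sending gameplay cues and quest events.
; While uncontested, a timer adds progress scaled by team count and accelerators.
; RepNotifies drive decal, radius FX, hologram parameter collection and stage lights, showing rival captures as contested locally.
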(event Custom|Cheat_ResetBeginPlayOnProgressReset
  ;[1 1 1] BeginPlay initialization
  (CallFunction|InitializeRadius)
  (Decal|InitializeDecal)
  (CallFunction|InitializeMaxAcceleratorsbySquadSize (Variables|/Script/FortniteGame.FortPlaylist|GetMaxSquadSize (stale EventBeginPlay _playlist)))
  (Networking|SwitchHasAuthority
    (:Authority
      (bind _bannerref (Variables|Default|GetBannerRef)
        (:then
          (Default|BindEventtoOnBannerRaised _bannerref (EventDispatchers|CreateEvent :!event Custom|HandleBannerRaised)))
        (:else)))))

(event Custom|POIClaimResetAllProgress
  ;[0 0.9262 1] Debug Cheats
  (State|SetState "NewEnumerator0")
  (Variables|Capture|SetCurrentCaptureTime 0.0)
  (Capture|EvaluatePawnsinCaptureZone)
  (Variables|/POIBanner/Actors/BGA_POIBanner_Banner.BGA_POIBanner_Banner_C|SetbIsRaised false (Variables|Default|GetBannerRef))
  (CallFunction|CheatResetBeginPlayonProgressReset))

(event EventBeginPlay
  ;[1 1 1] BeginPlay initialization
  (bind (_gamestate _playlist) (Class|FortAsyncActionCurrentPlaylistReady|PlaylistDataReadyAsync)
    (:OnReady
      (CallFunction|InitializeRadius)
      (Decal|InitializeDecal)
      (CallFunction|InitializeMaxAcceleratorsbySquadSize (Variables|/Script/FortniteGame.FortPlaylist|GetMaxSquadSize _playlist))
      (Networking|SwitchHasAuthority
        (:Authority
          (bind _bannerref (Variables|Default|GetBannerRef)
            (:then
              (Default|BindEventtoOnBannerRaised _bannerref (EventDispatchers|CreateEvent :!event Custom|HandleBannerRaised)))
            (:else)))))))

(event Custom|POIClaimCaptureAll
  ;[0 0.9262 1] Debug Cheats
  (CallFunction|UpdateBannerProgress (Variables|Capture|GetCapturingTeam) 1.0 true))

(event OnComponentBeginOverlap(CaptureZoneVolume)
  (params (OverlappedComponent :type Primitive%20Component%20Object%20Reference) (OtherActor :type Actor%20Object%20Reference) (OtherComp :type Primitive%20Component%20Object%20Reference) (OtherBodyIndex :type Integer) (bFromSweep :type Boolean) (SweepResult :type Hit%20Result%20Structure%20%28by%20ref%29))
  (bind _state (Variables|State|GetState))
  (bind _statetotagentry (Variables|Default|GetStateToTagEntry))
  (bind (_value _returnvalue) (Utilities|Map|Find _statetotagentry _state))
  ;[1 1 1] Overlap Handlers
  (Networking|SwitchHasAuthority
    (:Authority
      (bind _asfort_player_pawn (Utilities|Casting|CastToFortPlayerPawn OtherActor))
      (bind _returnvalue_2 (CallFunction|CanPawnCapture _asfort_player_pawn))
      (if _returnvalue_2
        ;;[0.4118 0.2902 1] If the banner is already raised, then call ShowCapturedPOIToast
        (if (CallFunction|ShouldShowCapturedPoiToast)
          (CallFunction|ShowCapturedPOIToast _asfort_player_pawn)
          (else
            ;;[0.4118 0.2902 1] If banner is not raised, then enable CapturePOI UI
            (CallFunction|OnPawnOverlapBegin _asfort_player_pawn)
            (CallFunction|TryGrantProgressTrackerAbility _asfort_player_pawn)
            (Variables|Default|SetLastEnteredPawn _asfort_player_pawn)
            (Utilities|Array|AddUnique (Variables|Capture|GetOverlappedPawns) _asfort_player_pawn)
            (Capture|EvaluatePawnsinCaptureZone)
            (Acceleration|EvaluateAcceleration)
            (Ability|GameplayEffect|AddLooseGameplayTags _asfort_player_pawn (Variables|Default|GetCapturing%20Point))
            (if _returnvalue
              (Ability|SendGameplayEventtoActor _asfort_player_pawn _value)
              (if (ScalableFloat|GetValueasBool (Variables|Default|GetShouldListenTeamChangedEvents))
                (bind _asfort_player_state_athena (Utilities|Casting|CastToFortPlayerStateAthena (Variables|/Script/Engine.Pawn|GetPlayerState _asfort_player_pawn)))
                (Player|BindEventtoOnSquadIdChangedDelegate _asfort_player_state_athena (EventDispatchers|CreateEvent :!event OnPawnTeamChanged))))))))))

(event OnComponentEndOverlap(CaptureZoneVolume)
  (params (OverlappedComponent :type Primitive%20Component%20Object%20Reference) (OtherActor :type Actor%20Object%20Reference) (OtherComp :type Primitive%20Component%20Object%20Reference) (OtherBodyIndex :type Integer))
  ;[1 1 1] Overlap Handlers
  (Networking|SwitchHasAuthority
    (:Authority
      (bind _asfort_player_pawn (Utilities|Casting|CastToFortPlayerPawn OtherActor))
      (CallFunction|OnPawnOverlapEnd _asfort_player_pawn)
      (bind _returnvalue (Utilities|Array|RemoveItem (Variables|Capture|GetOverlappedPawns) _asfort_player_pawn))
      (if _returnvalue
        (Variables|Default|SetLastLeftPawn _asfort_player_pawn)
        (Capture|EvaluatePawnsinCaptureZone)
        (Ability|GameplayEffect|RemoveLooseGameplayTags _asfort_player_pawn (Variables|Default|GetCapturing%20Point))
        (Acceleration|EvaluateAcceleration)
        (if (ScalableFloat|GetValueasBool (Variables|Default|GetShouldListenTeamChangedEvents))
          (Player|UnbindEventfromOnSquadIdChangedDelegate (stale OnComponentBeginOverlap(CaptureZoneVolume) _asfort_player_state_athena) (EventDispatchers|CreateEvent :!event OnPawnTeamChanged)))))))

(event Custom|FlashDecal
  (bind _decalmaterial (Variables|Decal|GetDecalMaterial))
  (Components|Timeline|SetPlayRate (Variables|BGA_POIBanner_CapturePoint|GetTL_DecalFlash) (Utilities|Reciprocal (Variables|Decal|Flash|GetDecalFlashTime)))
  (bind (_direction _emissivestrength) (Timeline|TL_DecalFlash PlayFromStart)
    (:Update
      (Rendering|Material|SetScalarParameterValue _decalmaterial "EmissiveStrength" (+ (Variables|Decal|Flash|GetDecalDefaultEmissiveStrength) (* (Variables|Decal|Flash|GetDecalFlashEmissive) _emissivestrength)) :!class /Script/Engine.MaterialInstanceDynamic)
      (Rendering|Material|SetScalarParameterValue _decalmaterial "OpacityScalar" (+ (Variables|Decal|GetDecalDefaultOpacity) _emissivestrength) :!class /Script/Engine.MaterialInstanceDynamic))))

(event Custom|ExecuteGameplayCueOnSelf
  (params (GameplayCueTag :type Gameplay%20Tag%20Structure) (Instigator :type Actor%20Object%20Reference))
  (Ability|ExecuteGameplayCueWithParamsReplicated self GameplayCueTag (Ability|GameplayCue|MakeGameplayCueParameters :NormalizedMagnitude 0.0 :RawMagnitude 0.0 :Location (Transformation|GetActorLocation) :Normal "0,0,0" :Instigator Instigator :GameplayEffectLevel 1 :AbilityLevel 1 :bReplicateLocationWhenUsingMinimalRepProxy false)))

(event Custom|AddLoopingGameplayCueOnSelf
  (params (GameplayCueTag :type Gameplay%20Tag%20Structure) (Instigator :type Actor%20Object%20Reference))
  (GameplayCue|AddGameplayCueOnActor(Looping) self GameplayCueTag (Ability|GameplayCue|MakeGameplayCueParameters :NormalizedMagnitude 0.0 :RawMagnitude 0.0 :Location (Transformation|GetActorLocation) :Normal "0,0,0" :Instigator Instigator :GameplayEffectLevel 1 :AbilityLevel 1 :bReplicateLocationWhenUsingMinimalRepProxy false)))

(event Custom|HandleBannerRaised
  (params (Team :type Integer) (Winners :type Array%20of%20Fort%20Player%20Pawn%20Object%20References) (Banner :type BGA%20POIBanner%20Banner%20Object%20Reference))
  (State|SetState "NewEnumerator3")
  (Variables|Default|SetCaptureMPCValue 1.0)
  (Variables|Default|SetLightsStage 2)
  (bind (_direction _newtrack_0) (Timeline|Timeline Play)
    (:Update
      (Variables|Default|SetCaptureAnimationMPCValue _newtrack_0))))

(event Custom|OnRep_CurrentCaptureProgress
  (bind _currentcaptureprogress (Variables|Capture|GetCurrentCaptureProgress))
  (Networking|SwitchHasAuthority
    (:Remote
      (Rendering|Material|SetScalarParameterValue "/Game/Environments/TechArt/Hologram/MPC_Hologram.MPC_Hologram" "ProgressNormalized" (Variables|Capture|GetCurrentCaptureProgress) :!class /Script/Engine.KismetMaterialLibrary)
      (if (> _currentcaptureprogress 0)
        (Rendering|Material|SetScalarParameterValue "/Game/Environments/TechArt/Hologram/MPC_Hologram.MPC_Hologram" "ScreensEnabled" 1.0 :!class /Script/Engine.KismetMaterialLibrary)
        (elif (<= _currentcaptureprogress 0)
          (Rendering|Material|SetScalarParameterValue "/Game/Environments/TechArt/Hologram/MPC_Hologram.MPC_Hologram" "ScreensEnabled" :!class /Script/Engine.KismetMaterialLibrary))))))

(fn CanPawnCapture
  (params (PlayerPawn :type Fort%20Player%20Pawn%20Object%20Reference))
  (bind (_returnvalue_x _returnvalue_y _returnvalue_z) (Math|Vector|BreakVector (Transformation|GetActorLocation PlayerPawn)))
  (bind (_returnvalue_x_2 _returnvalue_y_2 _returnvalue_z_2) (Math|Vector|BreakVector (Transformation|GetActorLocation)))
  (return (and (not (GameplayTags|HasMatchingGameplayTag PlayerPawn (Variables|Default|GetNonParticipantTag))) (or (>= _returnvalue_z _returnvalue_z_2) (Variables|Default|GetOptOutOfCaptureHeightCheck)))))

(fn ClientsOnStateChanged)

(fn DebugAcceleration
  (params (TeamCount :type Integer) (MaxCount :type Integer) (AccelerationPercent :type Float%20%28double-precision%29) (AccelerationScalar :type Float%20%28double-precision%29))
  (Development|PrintString (Utilities|String|ToString(Text) (Utilities|Text|FormatText "NSLOCTEXT(\"[546626A44F03BB80FF7002FD7FD766F9]\", \"F67AC50C424E976541D334A734E00D05\", \"Capturing Team Count: {1}\")" TeamCount)) true false "(R=0.000000,G=0.660000,B=1.000000,A=1.000000)" -1.0 "POICaptureTeamCount")
  (Development|PrintString (Utilities|String|ToString(Text) (Utilities|Text|FormatText "NSLOCTEXT(\"[546626A44F03BB80FF7002FD7FD766F9]\", \"E5EA7015498570FD33A1E7B4E140EC16\", \"Max Count: {1}\")" (Variables|GetMaxCount))) true false "(R=0.000000,G=0.660000,B=1.000000,A=1.000000)" -1.0 "POICaptureMaxCount")
  (Development|PrintString (Utilities|String|ToString(Text) (Utilities|Text|FormatText "NSLOCTEXT(\"[546626A44F03BB80FF7002FD7FD766F9]\", \"08D47DDC4E63FD72885ADF9986703EDF\", \"Acceleration Percent: {1}\")" (Variables|GetAccelerationPercent))) true false "(R=0.000000,G=0.660000,B=1.000000,A=1.000000)" -1.0 "POICaptureAccelPercent")
  (Development|PrintString (Utilities|String|ToString(Text) (Utilities|Text|FormatText "NSLOCTEXT(\"[546626A44F03BB80FF7002FD7FD766F9]\", \"1046DCAA4223CE2A9FD77B8948BCCE75\", \"Acceleration Scalar: {1}\")" (Variables|GetAccelerationScalar))) true false "(R=0.000000,G=0.660000,B=1.000000,A=1.000000)" -1.0 "POICaptureAccelScalar"))

(fn EndCaptureTimer
  (Utilities|Time|ClearandInvalidateTimerbyHandle (Variables|Capture|GetCaptureTimerHandle)))

(fn EvaluateAcceleration
  (Variables|Acceleration|SetCurrentAccelerationScalar (Acceleration|GetProgressAccelerationScalar)))

(fn EvaluatePawnsInCaptureZone
  (bind _overlappedpawns (Variables|Capture|GetOverlappedPawns))
  (bind _representedteams (LocalVariables|GetRepresentedTeams))
  (bind _representedteams_2 (LocalVariables|GetRepresentedTeams))
  (if (Utilities|Enum|Equal(Enum) (Variables|State|GetState) "NewEnumerator3")
    _
    (elif (Networking|IsServer)
      (if (> (Utilities|Array|Length _overlappedpawns) 0)
        (bind _array_element (Utilities|Array|ForEachLoopwithBreak _overlappedpawns)
          (:LoopBody
            ;[1 1 1] Track the unique teams currently represented by pawns in the capture zone
            (Utilities|Array|AddUnique _representedteams (Teams|TeamtoIndex (Team|GetActorTeam _array_element)))
            ;[1 1 1] As soon as we find a second team, we know the capture zone is contested and can stop checking
            (if (> (Utilities|Array|Length _representedteams) 1)
              _))
          (:Completed
            _))
        (Variables|Default|SetCurrentRepresentedTeams _representedteams_2)
        (if (== (Utilities|Array|Length _representedteams_2) 1)
          (Variables|Capture|SetLastCapturingTeam (Variables|Capture|GetCapturingTeam))
          (Variables|Capture|SetCapturingTeam (Utilities|Array|Get(acopy) (LocalVariables|GetRepresentedTeams)))
          (State|SetState "NewEnumerator1")
          (else
            (State|SetState "NewEnumerator2")))
        (else
          (State|SetState "NewEnumerator0"))))))

(fn GetCapturingTeamCount
  (bind (_value _returnvalue) (Utilities|Map|Find (Variables|Default|GetTeamIndexToNumMap) (Teams|IndextoTeam (Variables|Capture|GetCapturingTeam))))
  (return _returnvalue _value))

(fn GetProgressAccelerationScalar
  (bind _localnumaccelerators (LocalVariables|GetLocalNumAccelerators))
  (bind _returnvalue (- _localnumaccelerators 1))
  (bind _returnvalue_2 (Math|Conversions|ToFloat(Integer) _returnvalue))
  (bind (_teamfound _count) (CallFunction|GetCapturingTeamCount))
  (bind _row_maxaccelerationscalar (Variables|Acceleration|GetRow_MaxAccelerationScalar))
  (bind _returnvalue_3 (ScalableFloat|GetValueatLevel _row_maxaccelerationscalar))
  (bind _row_teamcountmultiplier (Variables|Acceleration|GetRow_TeamCountMultiplier))
  (bind _returnvalue_4 (ScalableFloat|GetValueatLevel _row_teamcountmultiplier _returnvalue_2))
  (if (CallFunction|IsTeamCountAccelerationEnabled)
    (if _teamfound
      (LocalVariables|SetLocalNumAccelerators (+ _count (Variables|Acceleration|GetNumAccelerators)))
      (if (ScalableFloat|GetValueasBool (Variables|Acceleration|GetRow_UseExplicitTeamCountMult))
        (CallFunction|DebugAcceleration 0 0 0.0 _returnvalue_4)
        (return _returnvalue_4)
        (else
          _))
      (else
        (LocalVariables|SetLocalNumAccelerators (Variables|Acceleration|GetNumAccelerators))))
    (bind _output_get (LocalVariables|SetEffectiveNumAccelerators (Math|Integer|Clamp(Integer) (LocalVariables|GetLocalNumAccelerators) 0 (Variables|Acceleration|GetMaxNumAccelerators))))
    (bind _output_get_2 (LocalVariables|SetAccelerationPercent (/ (Math|Conversions|ToFloat(Integer) _output_get) (Math|Conversions|ToFloat(Integer) (Variables|Acceleration|GetMaxNumAccelerators)))))
    (CallFunction|DebugAcceleration (LocalVariables|GetLocalNumAccelerators) (Variables|Acceleration|GetMaxNumAccelerators) _output_get_2 (+ (* _output_get_2 _returnvalue_3) 1.0))
    (return (+ (* _output_get_2 _returnvalue_3) 1.0))))

(fn GetStateChangeInstigator
  (params (NewState :type EPOIBannerState%20Enum :default "NewEnumerator0") (PrevState :type EPOIBannerState%20Enum :default "NewEnumerator0"))
  (switch Utilities|FlowControl|Switch|SwitchonEPOIBannerState NewState
    (:NewEnumerator0
      (return (Variables|Default|GetLastLeftPawn)))
    (:NewEnumerator1
      (if (Utilities|Enum|Equal(Enum) (Variables|State|GetPrevState) "NewEnumerator2")
        (return (Variables|Default|GetLastLeftPawn))
        (else
          (return (Variables|Default|GetLastEnteredPawn)))))
    (:NewEnumerator2
      (return (Variables|Default|GetLastEnteredPawn)))
    (:NewEnumerator3
      (return (Variables|Default|GetLastEnteredPawn)))))

(fn HandleAcceleratorBegin
  (params (Team :type Integer))
  (if (== Team (Variables|Capture|GetCapturingTeam))
    (Math|Integer|IncrementInt (Variables|Acceleration|GetNumAccelerators)
      (:%20%20
        (Acceleration|EvaluateAcceleration)))))

(fn HandleAcceleratorEnd
  (params (Team :type Integer))
  (Math|Integer|DecrementInt (Variables|Acceleration|GetNumAccelerators)
    (:%20%20
      (Acceleration|EvaluateAcceleration))))

(fn HandleCapturedStateBegin
  (Networking|SwitchHasAuthority
    (:Authority
      (Variables|Decal|SetDecalVisible false))))

(fn HandleContestedStateBegin
  (Networking|SwitchHasAuthority
    (:Authority
      (CallFunction|NotifyAllRepresentedTeamsProgressHalted))))

(fn HandleContestedStateEnd)

(fn HandleEmptyStateBegin
  (Networking|SwitchHasAuthority
    (:Authority
      (CallFunction|NotifyAllRepresentedTeamsProgressHalted))))

(fn HandleUncontestedStateBegin
  (Networking|SwitchHasAuthority
    (:Authority
      (Capture|StartCaptureTimer)
      (Progress|NotifyProgressStarted (Variables|Default|GetBannerRef) (Variables|Capture|GetCapturingTeam)))))

(fn HandleUncontestedStateEnd
  (Capture|EndCaptureTimer))

(fn InitializeAcceleratorListeners
  (bind _accelerators (Variables|Acceleration|GetAccelerators))
  (if (Utilities|Array|IsNotEmpty _accelerators)
    (for _array_element _accelerators
      (Utilities|IsValid _array_element
        (:Is%20Valid
          (Default|BindEventtoOnAccelerationBegin _array_element (EventDispatchers|CreateEvent :!event HandleAcceleratorBegin))
          (Default|BindEventtoOnAccelerationEnd _array_element (EventDispatchers|CreateEvent :!event HandleAcceleratorEnd)))))))

(fn InitializeDecal
  (bind _returnvalue (Rendering|Components|Decal|CreateDynamicMaterialInstance (Variables|Default|GetCaptureZoneDecal)))
  (bind _output_get (Variables|Decal|SetDecalMaterial _returnvalue))
  (bind _returnvalue_2 (Rendering|Material|GetScalarParameterValue _output_get "EmissiveStrength" :!class /Script/Engine.MaterialInstanceDynamic))
  (Variables|Decal|Flash|SetDecalDefaultEmissiveStrength _returnvalue_2)
  (bind _returnvalue_3 (Rendering|Material|GetScalarParameterValue _output_get "OpacityScalar" :!class /Script/Engine.MaterialInstanceDynamic))
  (Variables|Decal|SetDecalDefaultOpacity _returnvalue_3))

(fn InitializeMaxAcceleratorsBySquadSize
  (params (MaxSquadSize :type Integer))
  (Variables|Acceleration|SetMaxNumAccelerators MaxSquadSize))

(fn InitializeRadius
  (bind _capturezonevolume (Variables|Default|GetCaptureZoneVolume))
  (bind _returnvalue (Components|Capsule|GetUnscaledCapsuleRadius _capturezonevolume))
  ;[1 1 1] Use the proportion of the default radius to the radius from data to determine the scalar to be applied to everything that indicates radius
  (LocalVariables|SetScalar (/ (Math|Float|SelectFloat (Variables|Default|GetRadiusOverride) (ScalableFloat|GetValueatLevel (Variables|Default|GetRow_CaptureRadius)) (Variables|Default|GetbShouldOverrideRadius)) _returnvalue))
  (Components|Capsule|SetCapsuleRadius (Variables|Default|GetCaptureZoneVolume) (* _returnvalue (LocalVariables|GetScalar)) false)
  (Effects|Components|ParticleSystem|SetFloatParameter (Variables|Default|GetRadiusFX) "RadiusScale" (LocalVariables|GetScalar))
  (Transformation|SetWorldScale3D (Variables|Default|GetCaptureZoneDecal) (Math|Vector|MakeVector 1.0 (LocalVariables|GetScalar) (LocalVariables|GetScalar))))

(fn IsNewTeamCapturing
  (return (!= (Variables|Capture|GetLastCapturingTeam) (Variables|Capture|GetCapturingTeam))))

(fn IsTeamCountAccelerationEnabled
  (return (ScalableFloat|GetValueasBool (Variables|Acceleration|GetRow_TeamCountAccelerationEnabled))))

(fn NotifyAllRepresentedTeamsProgressHalted
  (for _array_element (Variables|Default|GetCurrentRepresentedTeams)
    (Progress|NotifyProgressHalted (Variables|Default|GetBannerRef) _array_element)))

(fn OnPawnOverlapBegin
  (params (Actor :type Actor%20Object%20Reference))
  (Utilities|FlowControl|Sequence
    (:then_0
      (bind _returnvalue (Utilities|GetPOIClaimControllerComponent Actor)
        (:Is%20Valid
          (Class|ControllerComponentPOIClaim|OnCapturePointEntered _returnvalue self (Variables|Default|GetBannerRef)))))
    (:then_1
      (CallFunction|UpdateTeamCountforNewActor Actor))))

(fn OnPawnOverlapEnd
  (params (Actor :type Actor%20Object%20Reference))
  (Utilities|FlowControl|Sequence
    (:then_0
      (bind _returnvalue (Utilities|GetPOIClaimControllerComponent Actor)
        (:Is%20Valid
          (Class|ControllerComponentPOIClaim|OnCapturePointExited _returnvalue self (Variables|Default|GetBannerRef)))))
    (:then_1
      (CallFunction|UpdateTeamCountforActorLeaving Actor))))

(fn OnPawnTeamChanged
  (params (PlayerState :type Fort%20Player%20State%20Athena%20Object%20Reference) (OldSquadIndex :type Byte))
  (CallFunction|UpdateTeamCountforNewActor (PlayerState|GetPawn PlayerState))
  (Capture|EvaluatePawnsinCaptureZone)
  (Acceleration|EvaluateAcceleration))

(fn OnRep_CaptureAnimationMPCValue
  (Rendering|Material|SetScalarParameterValue "/Game/Environments/TechArt/Hologram/MPC_Hologram.MPC_Hologram" "CaptureAnimPhase" (Variables|Default|GetCaptureAnimationMPCValue) :!class /Script/Engine.KismetMaterialLibrary))

(fn OnRep_CaptureMPCValue
  (Rendering|Material|SetScalarParameterValue "/Game/Environments/TechArt/Hologram/MPC_Hologram.MPC_Hologram" "Captured" (Variables|Default|GetCaptureMPCValue) :!class /Script/Engine.KismetMaterialLibrary))

(fn OnRep_CapturingTeam
  (bind _bannerref (Variables|Default|GetBannerRef))
  (Utilities|IsValid _bannerref
    (:Is%20Valid
      (Variables|/POIBanner/Actors/BGA_POIBanner_Banner.BGA_POIBanner_Banner_C|SetCurrentTeam (Variables|Capture|GetCapturingTeam) _bannerref))))

(fn OnRep_CurrentDecalColor
  (if (Networking|IsServer)
    _
    (elif (Utilities|Enum|Equal(Enum) (Variables|State|GetState) "NewEnumerator1")
      ;[1 1 1] On local clients, if the team currently capturing the point is not your team, show the contested color
      (LocalVariables|SetNewLocalVar (Fortnite|GetLocalFortPlayerControllers true))
      (bind _asfort_player_pawn (Utilities|Casting|CastToFortPlayerPawn (Pawn|GetControlledPawn (Fortnite|GetFirstLocalFortPlayerController true))))
      (if (== (Variables|Capture|GetCapturingTeam) (Teams|TeamtoIndex (Team|GetActorTeam _asfort_player_pawn)))
        (FX|SetDecalColor (Utilities|Map|Find (Variables|Decal|GetDecalColorMap) "NewEnumerator1"))
        (FX|SetRadiusFXColorParam 2)
        (FX|SetRadiusFXContestedParam false)
        (Class|BGAPOIBannerBanner|InformBannerState (Variables|Default|GetBannerRef) "NewEnumerator1")
        (else
          (FX|SetDecalColor (Utilities|Map|Find (Variables|Decal|GetDecalColorMap) "NewEnumerator2"))
          (FX|SetRadiusFXColorParam 1)
          (FX|SetRadiusFXContestedParam true)
          (Class|BGAPOIBannerBanner|InformBannerState (Variables|Default|GetBannerRef) "NewEnumerator2")))
      (CallFunction|FlashDecal)
      (switch Utilities|FlowControl|Switch|SwitchonEPOIBannerState (Variables|State|GetState)
        (:NewEnumerator0
          (FX|SetRadiusFXColorParam 0)
          (FX|SetRadiusFXContestedParam false)
          (Class|BGAPOIBannerBanner|InformBannerState (Variables|Default|GetBannerRef) "NewEnumerator0")
          (Rendering|Material|SetScalarParameterValue "/Game/Environments/TechArt/Hologram/MPC_Hologram.MPC_Hologram" "Contested" :!class /Script/Engine.KismetMaterialLibrary)
          (Rendering|Material|SetScalarParameterValue "/Game/Environments/TechArt/Hologram/MPC_Hologram.MPC_Hologram" "Captured" :!class /Script/Engine.KismetMaterialLibrary)
          (Rendering|Material|SetScalarParameterValue "/Game/Environments/TechArt/Hologram/MPC_Hologram.MPC_Hologram" "Uncontested" :!class /Script/Engine.KismetMaterialLibrary)
          (Rendering|Material|SetScalarParameterValue "/Game/Environments/TechArt/Hologram/MPC_Hologram.MPC_Hologram" "ScreensEnabled" 1.0 :!class /Script/Engine.KismetMaterialLibrary)
          (Variables|Default|SetLightsStage 1))
        (:NewEnumerator1
          (Rendering|Material|SetScalarParameterValue "/Game/Environments/TechArt/Hologram/MPC_Hologram.MPC_Hologram" "Contested" :!class /Script/Engine.KismetMaterialLibrary)
          (Rendering|Material|SetScalarParameterValue "/Game/Environments/TechArt/Hologram/MPC_Hologram.MPC_Hologram" "Captured" :!class /Script/Engine.KismetMaterialLibrary)
          (Rendering|Material|SetScalarParameterValue "/Game/Environments/TechArt/Hologram/MPC_Hologram.MPC_Hologram" "Uncontested" 1.0 :!class /Script/Engine.KismetMaterialLibrary)
          (Rendering|Material|SetScalarParameterValue "/Game/Environments/TechArt/Hologram/MPC_Hologram.MPC_Hologram" "ScreensEnabled" 1.0 :!class /Script/Engine.KismetMaterialLibrary)
          (Variables|Default|SetLightsStage 1))
        (:NewEnumerator2
          (FX|SetRadiusFXColorParam 1)
          (FX|SetRadiusFXContestedParam true)
          (Class|BGAPOIBannerBanner|InformBannerState (Variables|Default|GetBannerRef) "NewEnumerator2")
          (Rendering|Material|SetScalarParameterValue "/Game/Environments/TechArt/Hologram/MPC_Hologram.MPC_Hologram" "Contested" 1.0 :!class /Script/Engine.KismetMaterialLibrary)
          (Rendering|Material|SetScalarParameterValue "/Game/Environments/TechArt/Hologram/MPC_Hologram.MPC_Hologram" "Captured" :!class /Script/Engine.KismetMaterialLibrary)
          (Rendering|Material|SetScalarParameterValue "/Game/Environments/TechArt/Hologram/MPC_Hologram.MPC_Hologram" "Uncontested" 1.0 :!class /Script/Engine.KismetMaterialLibrary)
          (Rendering|Material|SetScalarParameterValue "/Game/Environments/TechArt/Hologram/MPC_Hologram.MPC_Hologram" "ScreensEnabled" 1.0 :!class /Script/Engine.KismetMaterialLibrary)
          (Variables|Default|SetLightsStage 3))
        (:NewEnumerator3
          (FX|SetRadiusFXColorParam 2)
          (FX|SetRadiusFXContestedParam false)
          (Class|BGAPOIBannerBanner|InformBannerState (Variables|Default|GetBannerRef) "NewEnumerator3")
          (Rendering|Material|SetScalarParameterValue "/Game/Environments/TechArt/Hologram/MPC_Hologram.MPC_Hologram" "Contested" :!class /Script/Engine.KismetMaterialLibrary)
          (Rendering|Material|SetScalarParameterValue "/Game/Environments/TechArt/Hologram/MPC_Hologram.MPC_Hologram" "Captured" 1.0 :!class /Script/Engine.KismetMaterialLibrary)
          (Rendering|Material|SetScalarParameterValue "/Game/Environments/TechArt/Hologram/MPC_Hologram.MPC_Hologram" "Uncontested" :!class /Script/Engine.KismetMaterialLibrary)
          (Rendering|Material|SetScalarParameterValue "/Game/Environments/TechArt/Hologram/MPC_Hologram.MPC_Hologram" "ScreensEnabled" 1.0 :!class /Script/Engine.KismetMaterialLibrary)
          (Variables|Default|SetLightsStage 2)))))
  (FX|SetDecalColor (Variables|Decal|GetCurrentDecalColor)))

(fn OnRep_DecalVisible
  (bind _decalvisible (Variables|Decal|GetDecalVisible))
  (Rendering|SetVisibility (Variables|Default|GetCaptureZoneDecal) _decalvisible false)
  (Rendering|SetVisibility (Variables|Default|GetRadiusFX) _decalvisible false))

(fn OnRep_LightsStage
  (bind _outactors (Actor|GetAllActorsOfClass "/Game/Environments/TechArt/Hologram/BP_CaptureStage_Lights.BP_CaptureStage_Lights_C"))
  (for _array_element _outactors
    (Variables|/Game/Environments/TechArt/Hologram/BP_CaptureStage_Lights.BP_CaptureStage_Lights_C|SetStage (Variables|Default|GetLightsStage) _array_element)))

(fn OnRep_PrevState
  (Networking|CanExecuteCosmeticEvents
    (:True
      (switch Utilities|FlowControl|Switch|SwitchonEPOIBannerState (Variables|State|GetPrevState)
        (:NewEnumerator1
          (State|HandleUncontestedStateEnd))
        (:NewEnumerator2
          (State|HandleContestedStateEnd))))))

(fn OnRep_State
  (Networking|CanExecuteCosmeticEvents
    (:True
      (CallFunction|ClientsonStateChanged)
      (switch Utilities|FlowControl|Switch|SwitchonEPOIBannerState (Variables|State|GetState)
        (:NewEnumerator0
          (State|HandleEmptyStateBegin))
        (:NewEnumerator1
          (State|HandleUncontestedStateBegin))
        (:NewEnumerator2
          (State|HandleContestedStateBegin))
        (:NewEnumerator3
          (State|HandleCapturedStateBegin))))))

(fn OnStateBegin
  (params (NewState :type EPOIBannerState%20Enum :default "NewEnumerator0") (Instigator :type Actor%20Object%20Reference))
  (Default|CallOnCaptureStateChanged self NewState (Variables|State|GetPrevState) (Variables|Capture|GetCapturingTeam))
  (Decal|SetDecalColorfromState NewState)
  (CallFunction|PlayStateBeginCue NewState Instigator)
  (switch Utilities|FlowControl|Switch|SwitchonEPOIBannerState NewState
    (:NewEnumerator0
      (State|HandleEmptyStateBegin))
    (:NewEnumerator1
      (State|HandleUncontestedStateBegin))
    (:NewEnumerator2
      (State|HandleContestedStateBegin))
    (:NewEnumerator3
      (State|HandleCapturedStateBegin))))

(fn OnStateEnd
  (params (OldState :type EPOIBannerState%20Enum :default "NewEnumerator0"))
  (bind _statecues (Variables|Default|GetStateCues))
  (bind (_value _returnvalue) (Utilities|Map|Find _statecues OldState))
  (Utilities|FlowControl|Sequence
    (:then_0
      (switch Utilities|FlowControl|Switch|SwitchonEPOIBannerState OldState
        (:NewEnumerator1
          (State|HandleUncontestedStateEnd))
        (:NewEnumerator2
          (State|HandleContestedStateEnd))))
    (:then_1
      (if (Utilities|Array|ContainsItem (Variables|Default|GetStatesWithLoopingCues) OldState)
        (if _returnvalue
          (Ability|RemoveGameplayCueReplicated _value))))))

(fn PlayStateBeginCue
  (params (State :type EPOIBannerState%20Enum%20%28by%20ref%29 :default "NewEnumerator0") (Instigator :type Actor%20Object%20Reference))
  (bind _statecues (Variables|Default|GetStateCues))
  (bind (_value _returnvalue) (Utilities|Map|Find _statecues State))
  (if _returnvalue
    (if (Utilities|Array|ContainsItem (Variables|Default|GetStatesWithLoopingCues) State)
      (CallFunction|AddLoopingGameplayCueonSelf _value Instigator)
      (else
        (CallFunction|ExecuteGameplayCueonSelf _value Instigator)))))

(fn SetDecalColor
  (params (Value :type Linear%20Color%20Structure))
  (bind _decalmaterial (Variables|Decal|GetDecalMaterial))
  (Rendering|Material|SetVectorParameterValue _decalmaterial "BaseColor" Value :!class /Script/Engine.MaterialInstanceDynamic)
  (Rendering|Material|SetVectorParameterValue _decalmaterial "EmissiveColor" Value :!class /Script/Engine.MaterialInstanceDynamic))

(fn SetDecalColorFromState
  (params (State :type EPOIBannerState%20Enum :default "NewEnumerator0"))
  (bind _decalcolormap (Variables|Decal|GetDecalColorMap))
  (bind (_value _returnvalue) (Utilities|Map|Find _decalcolormap State))
  (Utilities|FlowControl|Sequence
    (:then_0
      (if _returnvalue
        (Variables|Decal|SetCurrentDecalColor _value)))))

(fn SetRadiusFXColorParam
  (params (Param :type Integer))
  (Effects|Components|ParticleSystem|SetIntParameter (Variables|Default|GetRadiusFX) "Color" Param))

(fn SetRadiusFXContestedParam
  (params (Param :type Boolean))
  (Effects|Components|ParticleSystem|SetBoolParameter (Variables|Default|GetRadiusFX) "IsContested" Param))

(fn SetState
  (params (InState :type EPOIBannerState%20Enum :default "NewEnumerator0"))
  (bind _state (Variables|State|GetState))
  (bind _statetotag (Variables|Default|GetStateToTag))
  (bind (_value _returnvalue) (Utilities|Map|Find _statetotag _state))
  (bind _state_2 (Variables|State|GetState))
  (if (Utilities|Enum|NotEqual(Enum) InState _state_2)
    (Variables|State|SetPrevState _state_2)
    (Variables|State|SetState InState)
    (Networking|FlushNetDormancy)
    (State|OnStateEnd (Variables|State|GetPrevState))
    (bind _instigator (CallFunction|GetStateChangeInstigator (Variables|State|GetState) (Variables|State|GetPrevState)))
    (State|OnStateBegin _state _instigator)
    ;[1 1 1] Sends a gameplay event to every actor inside the capture point each time the status changes (for QS2 quests)
    (for _array_element (Variables|Capture|GetOverlappedPawns)
      (Utilities|IsValid _array_element
        (:Is%20Valid
          (if _returnvalue
            (Ability|SendGameplayEventtoActor _array_element _value)))))))

(fn ShouldShowCapturedPoiToast
  (return (Variables|/POIBanner/Actors/BGA_POIBanner_Banner.BGA_POIBanner_Banner_C|GetbIsRaised (Variables|Default|GetBannerRef))))

(fn ShowCapturedPOIToast
  (params (Actor :type Actor%20Object%20Reference))
  (bind _returnvalue (Utilities|GetPOIClaimControllerComponent Actor)
    (:Is%20Valid
      (Class|ControllerComponentPOIClaim|ShowCapturedPOIToast _returnvalue (Variables|/POIBanner/Actors/BGA_POIBanner_Banner.BGA_POIBanner_Banner_C|GetWinningTeam (Variables|Default|GetBannerRef)) (Variables|/POIBanner/Actors/BGA_POIBanner_Banner.BGA_POIBanner_Banner_C|GetWinner%20Pawns (Variables|Default|GetBannerRef)) (Variables|Default|GetBannerRef)))))

(fn StartCaptureTimer
  (bind _returnvalue (Utilities|Time|SetTimerbyFunctionName 0 "UpdateCaptureProgress" (Variables|Capture|GetCaptureRate) true))
  (Variables|Capture|SetCaptureTimerHandle _returnvalue))

(fn TryGrantProgressTrackerAbility
  (params (Actor :type Actor%20Object%20Reference))
  (Utilities|IsValid Actor
    (:Is%20Valid
      (Progress|TryAddParticipant (Variables|Default|GetBannerRef) Actor))))

(fn UpdateCaptureProgress
  (bind _timetocapture (Variables|Capture|GetTimeToCapture))
  (bind _returnvalue (ScalableFloat|GetValueatLevel _timetocapture))
  (bind _currentaccelerationscalar (Variables|Acceleration|GetCurrentAccelerationScalar))
  (bind _capturerate (Variables|Capture|GetCaptureRate))
  (bind _returnvalue_2 (* _capturerate _currentaccelerationscalar))
  (bind _output_get (Variables|Capture|SetCurrentCaptureTime (+ (Variables|Capture|GetCurrentCaptureTime) _returnvalue_2)))
  (Variables|Capture|SetCurrentCaptureProgress (/ _output_get _returnvalue))
  (CallFunction|UpdateBannerProgress (Variables|Capture|GetCapturingTeam) (/ _returnvalue_2 _returnvalue)))

(fn UpdateTeamCountForActorLeaving
  (params (Actor :type Actor%20Object%20Reference))
  (bind _teamindextonummap (Variables|Default|GetTeamIndexToNumMap))
  (bind _output_get (LocalVariables|SetActorTeam (Team|GetActorTeam Actor)))
  (if (.ReturnValue (Utilities|Map|Find _teamindextonummap _output_get))
    (Utilities|Map|Add (Variables|Default|GetTeamIndexToNumMap) (LocalVariables|GetActorTeam) (Math|Integer|Clamp(Integer) (- (Utilities|Map|Find _teamindextonummap _output_get) 1) 0 4))))

(fn UpdateTeamCountForNewActor
  (params (Actor :type Actor%20Object%20Reference))
  (bind _teamindextonummap (Variables|Default|GetTeamIndexToNumMap))
  (bind _output_get (LocalVariables|SetActorTeam (Team|GetActorTeam Actor)))
  (if (.ReturnValue (Utilities|Map|Find _teamindextonummap _output_get))
    ;[1 1 1] If we already have an entry for this team, increment the count by 1
    (Utilities|Map|Add (Variables|Default|GetTeamIndexToNumMap) (LocalVariables|GetActorTeam) (+ (Utilities|Map|Find _teamindextonummap _output_get) 1))
    (else
      ;[1 1 1] If we don't already have an entry for this team, create a new one and set it to 1
      (Utilities|Map|Add (Variables|Default|GetTeamIndexToNumMap) (LocalVariables|GetActorTeam) 1))))

(fn ConstructionScript)
