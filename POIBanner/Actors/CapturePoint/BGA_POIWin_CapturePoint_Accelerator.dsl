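; Capture point accelerator actor for POI banners that tracks whether acceleration is active and announces changes.
; BeginAcceleration and EndAcceleration set the AccelerationActive flag and fire begin or end dispatchers with the team index.
; The flag's RepNotify passes the new state to RepAccelerationActive, an empty hook apparently meant for child overrides, then fires OnRepAccelerationUpdated.
; No capture rate or timing logic lives here.
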(fn BeginAcceleration
  (params (Team :type Integer))
  (Variables|Default|SetAccelerationActive true)
  (Default|CallOnAccelerationBegin Team))

(fn EndAcceleration
  (params (Team :type Integer))
  (Variables|Default|SetAccelerationActive false)
  (Default|CallOnAccelerationEnd Team))

(fn OnRep_AccelerationActive
  (bind _accelerationactive (Variables|Default|GetAccelerationActive))
  (CallFunction|RepAccelerationActive _accelerationactive)
  (Default|CallOnRepAccelerationUpdated _accelerationactive))

(fn RepAccelerationActive
  (params (IsActive :type Boolean)))

(fn ConstructionScript)
