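; Bounds actor for the POI banner claim capture point that reports pawns entering and leaving its area.
; Actor begin and end overlap events cast the other actor to FortPawn and pass it to the OnActorEnteredBounds and OnActorExitBounds dispatchers.
; It holds no claim logic itself, and its BeginPlay, Tick and construction script are empty.
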(event EventBeginPlay)

(event Collision|EventActorBeginOverlap
  (params (OtherActor :type Actor%20Object%20Reference))
  (bind _asfort_pawn (Utilities|Casting|CastToFortPawn OtherActor))
  (Default|CallOnActorEnteredBounds _asfort_pawn))

(event Collision|EventActorEndOverlap
  (params (OtherActor :type Actor%20Object%20Reference))
  (bind _asfort_pawn (Utilities|Casting|CastToFortPawn OtherActor))
  (Default|CallOnActorExitBounds _asfort_pawn))

(event EventTick
  (params (DeltaSeconds :type Float%20%28single-precision%29)))

(fn ConstructionScript)
